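# Reticle widget for the WinterDecay teleport weapon, extending the recharge-stack reticle base with a material-driven image and charge pips.
# BindToWeaponEvents binds or unbinds projectile, marker, cooldown and ammo delegates.
# Projectile, marker and teleport-cooldown events play animations forward or reversed, driving material parameters and pip opacity.
# A delta gate toggles a cooldown timer when ammo crosses full clip, feeding recharge time to the pips; primary-fire cooldown handlers are empty.
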WidgetBlueprint /WinterDecay/UI/WBP_WinterDecay_Reticle
animations [
 Anim_Teleport_Throw {
  binding_count=1
  length_seconds=0.7
  object=&self:Anim_Teleport_Throw
  sequence {bindings [Image_Reticle {
     id=#1
     possessed_class=Image
     tracks [{
       type=MovieSceneWidgetMaterialTrack
       sections [{
         type=MovieSceneParameterSection
         end_frame=14
         start_frame=0
         channels [
          SpiralScale {type=MovieSceneScriptingFloatChannel, keys [[0,0.5],[5,0.56],[10,0.45],[14,0.5]]}
          Switch {type=MovieSceneScriptingFloatChannel, keys [[0,0],[10,0.5],[14,1]]}
          ShadowSpacing {type=MovieSceneScriptingFloatChannel, keys [[0,0.3],[13.9997,0]]}
          GlowShadowIntensity {type=MovieSceneScriptingFloatChannel, keys [[0,0.2],[13.9997,1]]}
          InnerSpiralRotation {type=MovieSceneScriptingFloatChannel, keys [[0,0],[5,-0.2158],[14,-0.5]]}
          OuterSpiralRotation {type=MovieSceneScriptingFloatChannel, keys [[0,0],[5,-0.045],[14,-0.5]]}
          InnerSpiralScale {type=MovieSceneScriptingFloatChannel, keys [[0,0.5],[14,0.25]]}
          InnerSpiralOpacity {type=MovieSceneScriptingFloatChannel, keys [[0,0],[5,0.5],[14,0.8]]}
         ]
        }]
      }]
    }]}
 }
 Anim_Teleport_UseTeleport {
  binding_count=1
  length_seconds=0.6
  object=&self:Anim_Teleport_UseTeleport
  sequence {bindings [Image_Reticle {
     id=#1
     possessed_class=Image
     tracks [{
       type=MovieSceneWidgetMaterialTrack
       sections [{
         type=MovieSceneParameterSection
         end_frame=12
         start_frame=0
         channels [
          IsPortalActive {type=MovieSceneScriptingFloatChannel, keys [[0,1],[2,1.12],[12,0]]}
          OuterSpiralRotation {type=MovieSceneScriptingFloatChannel, keys [[0,0],[12,-0.5]]}
          ScaleXDiamond {type=MovieSceneScriptingFloatChannel, keys [[0,0.6],[3,0.8696],[6,0.2296]]}
          ScaleYDiamond {type=MovieSceneScriptingFloatChannel, keys [[0,0.6],[3,0.69],[6,0.05]]}
          GlowShadowIntensity {type=MovieSceneScriptingFloatChannel, keys [[0,0.5],[2,1.536],[6,0.3]]}
          Diamond Opacity {type=MovieSceneScriptingFloatChannel, keys [[4,1],[6,0]]}
          SpeedRotation {type=MovieSceneScriptingFloatChannel, const [0,1]}
          SpiralColorR {type=MovieSceneScriptingFloatChannel, const [0.0887,1]}
          SpiralColorG {type=MovieSceneScriptingFloatChannel, const [0.4125,1]}
          SpiralColorB {type=MovieSceneScriptingFloatChannel, const [0.8714,1]}
          SpiralColorA {type=MovieSceneScriptingFloatChannel}
          DiamondColorR {type=MovieSceneScriptingFloatChannel, keys [[0,0.5029],[3,0],[5,0.0887]]}
          DiamondColorG {type=MovieSceneScriptingFloatChannel, keys [[0,1],[3,1],[5,0.4125]]}
          DiamondColorB {type=MovieSceneScriptingFloatChannel, keys [[0,1],[3,1],[5,0.8714]]}
          DiamondColorA {type=MovieSceneScriptingFloatChannel, const [1,1]}
         ]
        }]
      }]
    }]}
 }
 Anim_Teleport_CooldownEnded {
  binding_count=1
  length_seconds=0.3
  object=&self:Anim_Teleport_CooldownEnded
  sequence {bindings [Image_Reticle {
     id=#1
     possessed_class=Image
     tracks [{
       type=MovieSceneWidgetMaterialTrack
       sections [{
         type=MovieSceneParameterSection
         end_frame=6
         start_frame=0
         channels [
          SpiralScale {type=MovieSceneScriptingFloatChannel, keys [[0,0.5],[1,0.6],[4,0.45],[6,0.5]]}
          Switch {type=MovieSceneScriptingFloatChannel, const [0,2]}
          DropLinesOpacity {type=MovieSceneScriptingFloatChannel, keys [[-0.0003,1],[0,0.2],[6,1]]}
          ShadowSpacing {type=MovieSceneScriptingFloatChannel, keys [[-0.0003,0.3],[0,1],[6,0.3]]}
          GlowShadowIntensity {type=MovieSceneScriptingFloatChannel, const [1,2]}
          InnerSpiralRotation {type=MovieSceneScriptingFloatChannel, const [-1,2]}
          OuterSpiralRotation {type=MovieSceneScriptingFloatChannel, keys [[-0.0003,-1],[6,-1.5]]}
          InnerSpiralScale {type=MovieSceneScriptingFloatChannel, const [0.63,2]}
          OuterSpiralOpacity {type=MovieSceneScriptingFloatChannel, keys [[0,0.2],[6,1]]}
         ]
        }]
      }]
    }]}
 }
 Anim_Teleport_OnCooldown {
  binding_count=1
  length_seconds=0.25
  object=&self:Anim_Teleport_OnCooldown
  sequence {bindings [Image_Reticle {
     id=#1
     possessed_class=Image
     tracks [{
       type=MovieSceneWidgetMaterialTrack
       sections [{
         type=MovieSceneParameterSection
         end_frame=5
         start_frame=-1
         channels [
          SpiralScale {type=MovieSceneScriptingFloatChannel, const [0.5,2]}
          Switch {type=MovieSceneScriptingFloatChannel, const [0,2]}
          DropLinesOpacity {type=MovieSceneScriptingFloatChannel, keys [[-0.0003,1],[-0.0003,0.2],[5,0.2]]}
          ShadowSpacing {type=MovieSceneScriptingFloatChannel, const [1,2]}
          GlowShadowIntensity {type=MovieSceneScriptingFloatChannel, const [1,2]}
          InnerSpiralRotation {type=MovieSceneScriptingFloatChannel, const [-1,2]}
          OuterSpiralRotation {type=MovieSceneScriptingFloatChannel, const [-1,2]}
          InnerSpiralScale {type=MovieSceneScriptingFloatChannel, const [0.63,2]}
          OuterSpiralOpacity {type=MovieSceneScriptingFloatChannel, const [0.2,2]}
         ]
        }]
      }]
    }]}
 }
 Anim_TeleportThrow {
  binding_count=1
  length_seconds=0.6
  object=&self:Anim_TeleportThrow
  sequence {bindings [Image_Reticle {
     id=#2
     possessed_class=Image
     tracks [{
       type=MovieSceneWidgetMaterialTrack
       sections [{
         type=MovieSceneParameterSection
         end_frame=12
         start_frame=0
         channels [
          IsPortalActive {type=MovieSceneScriptingFloatChannel, keys [[0,0],[3,-0.24],[6,0],[10,1]]}
          OuterSpiralRotation {type=MovieSceneScriptingFloatChannel, keys [[0,0],[9,-0.7]]}
          ScaleXDiamond {type=MovieSceneScriptingFloatChannel, keys [[0,0],[7,0.166],[10,0.6]]}
          ScaleYDiamond {type=MovieSceneScriptingFloatChannel, keys [[0,0],[7,0.02],[10,0.4],[12,0.6]]}
          SpeedRotation {type=MovieSceneScriptingFloatChannel, keys [[9,0],[10,0.3]]}
          GlowShadowIntensity {type=MovieSceneScriptingFloatChannel, keys [[0,0.3],[8,0.3],[9,0.5]]}
          Diamond Opacity {type=MovieSceneScriptingFloatChannel, keys [[6,0],[7,1]]}
          SpiralColorR {type=MovieSceneScriptingFloatChannel, keys [[9,1],[10,0.0887]]}
          SpiralColorG {type=MovieSceneScriptingFloatChannel, keys [[9,1],[10,0.4125]]}
          SpiralColorB {type=MovieSceneScriptingFloatChannel, keys [[9,1],[10,0.8714]]}
          SpiralColorA {type=MovieSceneScriptingFloatChannel, const [1,1]}
          DiamondColorR {type=MovieSceneScriptingFloatChannel, keys [[0,0],[10,0.5029]]}
          DiamondColorG {type=MovieSceneScriptingFloatChannel, const [1,2]}
          DiamondColorB {type=MovieSceneScriptingFloatChannel, const [1,2]}
          DiamondColorA {type=MovieSceneScriptingFloatChannel, const [1,1]}
         ]
        }]
      }]
    }]}
 }
 Anim_TeleportUnavailable {
  binding_count=2
  length_seconds=0.25
  object=&self:Anim_TeleportUnavailable
  sequence {bindings [
    Image_Reticle {
     id=#3
     possessed_class=Image
     tracks [{
       type=MovieSceneWidgetMaterialTrack
       sections [{
         type=MovieSceneParameterSection
         end_frame=5
         start_frame=0
         channels [
          IsPortalActive {type=MovieSceneScriptingFloatChannel, const [1,1]}
          GlobalOpacity {type=MovieSceneScriptingFloatChannel, keys [[0,1],[5,0.3]]}
          ScaleXDiamond {type=MovieSceneScriptingFloatChannel, const [0.6,1]}
          ScaleYDiamond {type=MovieSceneScriptingFloatChannel, const [0.6,1]}
          GlowShadowIntensity {type=MovieSceneScriptingFloatChannel, const [0.5,1]}
          SpiralColorR {type=MovieSceneScriptingFloatChannel, const [0.0887,1]}
          SpiralColorG {type=MovieSceneScriptingFloatChannel, const [0.4125,1]}
          SpiralColorB {type=MovieSceneScriptingFloatChannel, const [0.8714,1]}
          SpiralColorA {type=MovieSceneScriptingFloatChannel, const [1,1]}
          DiamondColorR {type=MovieSceneScriptingFloatChannel, const [0.5029,1]}
          DiamondColorG {type=MovieSceneScriptingFloatChannel, const [1,1]}
          DiamondColorB {type=MovieSceneScriptingFloatChannel, const [1,1]}
          DiamondColorA {type=MovieSceneScriptingFloatChannel, const [1,1]}
         ]
        }]
      }]
    }
    WBP_WinterDecay_Reticle_PipCollection {
     id=#4
     possessed_class=WBP_WinterDecay_Reticle_PipCollection_C
     tracks [{
       type=MovieSceneFloatTrack
       display_name=Render Opacity
       sections [{type=MovieSceneFloatSection, end_frame=5, start_frame=0, channels [None {type=MovieSceneScriptingFloatChannel, keys [[0,1],[5,0.5]]}]}]
      }]
    }
   ]}
 }
 Anim_TeleportCooldown {
  binding_count=2
  length_seconds=0.2
  object=&self:Anim_TeleportCooldown
  sequence {bindings [
    Image_Reticle {
     id=#5
     possessed_class=Image
     tracks [{
       type=MovieSceneWidgetMaterialTrack
       sections [{
         type=MovieSceneParameterSection
         end_frame=4
         start_frame=0
         channels [GlobalOpacity {type=MovieSceneScriptingFloatChannel, keys [[0,1],[4,0.5]]}]
        }]
      }]
    }
    WBP_WinterDecay_Reticle_PipCollection {
     id=#6
     possessed_class=WBP_WinterDecay_Reticle_PipCollection_C
     tracks [{
       type=MovieSceneFloatTrack
       display_name=Render Opacity
       sections [{type=MovieSceneFloatSection, end_frame=4, start_frame=0, channels [None {type=MovieSceneScriptingFloatChannel, keys [[0,1],[4,0.5]]}]}]
      }]
    }
   ]}
 }
]
graphs [
 BindToWeaponEvents {nodes [
   n1 = BindToWeaponEvents!() -> (then, DoBind:bool, EquippedWeapon:Fort Weapon Object Reference) {is=title K2Node_FunctionEntry}
   > n2 = Cast To B_WinterDecay_Teleport!(execute, Object=@n1.EquippedWeapon) -> (then, CastFailed, AsB Winter Decay Teleport:B Winter Decay Teleport Object Reference) {is=title K2Node_DynamicCast}
   > Set WinterDecay_WeaponRef(WinterDecay_WeaponRef:B Winter Decay Teleport Object Reference=@n2) -> (Output_Get:B Winter Decay Teleport Object Reference)
   > n3 = Branch(Condition=Get DoBind() -> (DoBind:bool))
   n4 = Create Event()
   n5 = Get WinterDecay_WeaponRef() -> (WinterDecay_WeaponRef:B Winter Decay Teleport Object Reference)
   > Bind Event to On Local Ammo Changed(self:Fort Weapon Object Reference=@n5, Delegate=@n4) {is=title K2Node_AddDelegate}
   n6 = Create Event()
   > Bind Event to On New Projectile Created and Valid(self:B Winter Decay Teleport Object Reference=@n5, Delegate=@n6) {is=title K2Node_AddDelegate}
   n7 = Create Event()
   > Bind Event to On Primary Fire Cooldown(self:B Winter Decay Teleport Object Reference=@n5, Delegate=@n7) {is=title K2Node_AddDelegate}
   n8 = Create Event()
   > Bind Event to On Primary Fire Cooldown Ended(self:B Winter Decay Teleport Object Reference=@n5, Delegate=@n8) {is=title K2Node_AddDelegate}
   n9 = Create Event()
   > Bind Event to On Teleported Cooldown(self:B Winter Decay Teleport Object Reference=@n5, Delegate=@n9) {is=title K2Node_AddDelegate}
   n10 = Create Event()
   > Bind Event to On Teleported Cooldown Ended(self:B Winter Decay Teleport Object Reference=@n5, Delegate=@n10) {is=title K2Node_AddDelegate}
   n11 = Create Event()
   > Bind Event to On New BGACreated and Valid(self:B Winter Decay Teleport Object Reference=@n5, Delegate=@n11) {is=title K2Node_AddDelegate}
   Unbind Event from On Local Ammo Changed(execute=@n3.else, self:Fort Weapon Object Reference=@n5, Delegate=@n4) {is=title K2Node_RemoveDelegate}
   > Unbind Event from On New Projectile Created and Valid(self:B Winter Decay Teleport Object Reference=@n5, Delegate=@n6) {is=title K2Node_RemoveDelegate}
   > Unbind Event from On Primary Fire Cooldown(self:B Winter Decay Teleport Object Reference=@n5, Delegate=@n7) {is=title K2Node_RemoveDelegate}
   > Unbind Event from On Primary Fire Cooldown Ended(self:B Winter Decay Teleport Object Reference=@n5, Delegate=@n8) {is=title K2Node_RemoveDelegate}
   > Unbind Event from On Teleported Cooldown(self:B Winter Decay Teleport Object Reference=@n5, Delegate=@n9) {is=title K2Node_RemoveDelegate}
   > Unbind Event from On Teleported Cooldown Ended(self:B Winter Decay Teleport Object Reference=@n5, Delegate=@n10) {is=title K2Node_RemoveDelegate}
   > Unbind Event from On New BGACreated and Valid(self:B Winter Decay Teleport Object Reference=@n5, Delegate=@n11) {is=title K2Node_RemoveDelegate}
  ]}
 EventGraph {nodes [
   OnProjectileCreated!() -> (OutputDelegate, then, TeleportProjectile:B Prj Winter Decay Object Reference) {is=title K2Node_CustomEvent}
   > PlayAnimationForward(InAnimation=Get Anim_TeleportThrow() -> (Anim_TeleportThrow:Widget Animation Object Reference))
   OnTeleportedCooldown() {is=title K2Node_CustomEvent}
   > PlayAnimationForward(InAnimation=Get Anim_Teleport_UseTeleport() -> (Anim_Teleport_UseTeleport:Widget Animation Object Reference))
   > PlayAnimationForward(InAnimation=Get Anim_TeleportCooldown() -> (Anim_TeleportCooldown:Widget Animation Object Reference))
   OnTeleportedCooldownEnded() {is=title K2Node_CustomEvent}
   > PlayAnimationReverse(InAnimation=Get Anim_TeleportCooldown())
   n1 = OnBGAMarkerUpdated!() -> (OutputDelegate, then, TeleportMarker BGA:BGA Demon Mask Void Teleport Base Object Reference) {is=title K2Node_CustomEvent}
   > n2 = Is Valid(InputObject=@n1.TeleportMarker BGA)
   > PlayAnimationReverse(InAnimation=Get Anim_TeleportUnavailable() -> (Anim_TeleportUnavailable:Widget Animation Object Reference))
   PlayAnimationReverse(execute=@n2.Is Not Valid, InAnimation=Get Anim_TeleportThrow())
   Event Destruct()
   > Parent: Destruct()
   > HandleCooldownTimer!(execute, self:Fort Weapon Reticle Extension Widget Base Object Reference, bActivateTimer:bool=false, TimerInterval:float=0) -> (then)
   Event Construct()
   > Parent: Construct()
   > Set IsCooldownTimerActive(IsCooldownTimerActive:bool=false) -> (Output_Get:bool)
   > UpdateCooldownQueryTimer(self:Fort Weapon Reticle Extension Widget Base Object Reference)
   n3 = Event OnLocalAmmoChanged!() -> (OutputDelegate, then, LocalCount:int=0, LocalRemaining:int=0) {is=title K2Node_Event}
   > Parent: OnLocalAmmoChanged!(execute, LocalCount:int=@n3.LocalCount, LocalRemaining:int=@n3.LocalRemaining) -> (then) {is=title K2Node_CallParentFunction}
   > UpdateCooldownQueryTimer()
   Event UpdateCooldownQueryTimer() {is=title K2Node_Event}
   > n4 = Get!(execute:Exec) -> (then:Exec, else:Exec, WinterDecay_WeaponRef:B Winter Decay Teleport Object Reference)
   n5 = GetLocalMagazineAmmoCount!(self:Fort Weapon Object Reference=@n4) -> (ReturnValue:int=0)
   n6 = GetBulletsPerClip!(self:Fort Weapon Object Reference=@n4) -> (ReturnValue:int=0)
   > n7 = Delta Gate Wildcard!(
    execute:Exec
    Variable:"Boolean (by ref)"=Get IsCooldownTimerActive() -> (IsCooldownTimerActive:bool)
    New Value:bool=NOT Boolean(A="integer >= integer"(A=@n5, B=@n6))
   ) -> (Success:Exec, Updated Variable:"Boolean (by ref)", Previous Variable:bool=~) {is=title K2Node_MacroInstance}
   > n8 = Branch(Condition=@n7)
   > HandleCooldownTimer(bActivateTimer=true, TimerInterval=0.1)
   HandleCooldownTimer(execute=@n8.else, bActivateTimer=false, TimerInterval=0)
   Event UpdateRemainingCooldown() {is=title K2Node_Event}
   n9 = GetCooldownTimeRemaining!(self:Fort Weapon Reticle Extension Widget Base Object Reference) -> (RemainingTime:float, TotalDuration:float)
   > UpdateRemainingRechargeTime!(
    execute
    self:Fort Weapon Reticle Recharge Stack Base Pip Collection Base Object Reference=Get WBP_WinterDecay_Reticle_PipCollection() -> (WBP_WinterDecay_Reticle_PipCollection:WBP Winter Decay Reticle Pip Collection Object Reference)
    TimeRemaining:float=@n9
    TotalDuration:float=@n9.TotalDuration
    PipPercentMatParamName:Name=Percent
    ShouldSetPipColor:bool=false
    NotReadyColor:Linear Color Structure
    ReadyColor:Linear Color Structure
    PipColorMatParamName:Name=None
   ) -> (then)
   PrintString!(execute, InString=FireCooldownEnded, bPrintToScreen, bPrintToLog, TextColor, Duration, Key) -> (then)
   PrintString(InString=CooldownStarted)
   OnPrimaryFireCooldown() {is=title K2Node_CustomEvent}
   OnPrimaryFireCooldownEnded() {is=title K2Node_CustomEvent}
  ]}
 GetPipCollection {nodes [
   GetPipCollection() {is=title K2Node_FunctionEntry}
   > Return Node!(execute, PipCollection:WBP Reticle Recharge Stack Base Pip Collection Object Reference=Get WBP_WinterDecay_Reticle_PipCollection())
  ]}
]
variables [
 IsCooldownTimerActive {type=boolean}
 WinterDecay_WeaponRef {type=B_WinterDecay_Teleport_C, default=null}
]
widgets {
 info {
  inheritedWidgetCount=1
  parentClass=/Game/Athena/HUD/reticles/ReticleExtensionWidgets/ChargeStack/WBP_Reticle_RechargeStackBase.WBP_Reticle_RechargeStackBase_C
  rootWidgetClass=&Overlay
  widgetCount=5
 }
 widgets [
  Overlay_0 {widgetClassPath=&Overlay, widget=Overlay_0}
  Image_Reticle {
   widgetClassPath=&Image
   bIsVariable=true
   parent=Overlay_0
   widget=Image_Reticle
   properties {brush {
     bIsDynamicallyLoaded=false
     drawAs=Image
     imageType=FullColor
     mirroring=NoMirror
     resourceName=None
     resourceObject=/WinterDecay/UI/Materials/M_UI_WinterDecay_Reticle
     tiling=NoTile
     imageSize=<130 130>
     margin=<ltrb 0 0 0 0>
     outlineSettings {
      bUseBrushTransparency=false
      roundingType=HalfHeightRadius
      width=0
      color {colorUseRule=UseColor_Specified, specifiedColor=<rgba 0 0 0 0>}
      cornerRadii=<0 0 0 0>
     }
     tintColor {colorUseRule=UseColor_Specified, specifiedColor=<rgba 1 1 1 1>}
     uVRegion {bIsValid=false, max=<0 0>, min=<0 0>}
    }}
   slot {class=OverlaySlot, properties {horizontalAlignment=HAlign_Center, verticalAlignment=VAlign_Center}}
  }
  FortVisualAttachment_56 {
   widgetClassPath=&FortVisualAttachment
   parent=Overlay_0
   widget=FortVisualAttachment_56
   properties {contentAnchor=<0.4 0.7186>}
   slot {class=OverlaySlot, properties {horizontalAlignment=HAlign_Center, verticalAlignment=VAlign_Center}}
  }
  WBP_WinterDecay_Reticle_PipCollection {
   widgetClassPath=&/WBP_WinterDecay_Reticle_PipCollection
   bIsVariable=true
   parent=FortVisualAttachment_56
   widget=WBP_WinterDecay_Reticle_PipCollection
   properties {bOverride_Cursor=true}
   slot {class=SizeBoxSlot}
  }
  TutorialHighlight {widgetClassPath=&TutorialHighlight}
 ]
}
