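# WinterDecay reticle pip widget showing one charge of a recharge stack, built on the shared RechargeStack pip base.
# Overlays a material-driven pip with a radial fill image registered as the base's pip image, visible while the charge is empty.
# InitFillState snaps fill percent, material parameters and opacity to the filled or empty look.
# Gain and Use animations, each stopping the other, keyframe the pip's stack, circle and glow parameters and toggle the radial's opacity.
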WidgetBlueprint /WinterDecay/UI/WBP_WinterDecay_Reticle_Pip
animations [
 Anim_Use {
  binding_count=2
  length_seconds=0.65
  object=&self:Anim_Use
  sequence {bindings [
    Image_PipFill {
     id=#1
     possessed_class=Image
     tracks [{
       type=MovieSceneFloatTrack
       display_name=Render Opacity
       sections [{
         type=MovieSceneFloatSection
         end_frame=13
         start_frame=0
         channels [None {type=MovieSceneScriptingFloatChannel, keys [[0,0],[11,0],[12,1],[13,1]]}]
        }]
      }]
    }
    Image_Pip {
     id=#2
     possessed_class=Image
     tracks [{
       type=MovieSceneWidgetMaterialTrack
       sections [{
         type=MovieSceneParameterSection
         channels [
          Stack Opacity {type=MovieSceneScriptingFloatChannel, thinned_from=4, keys [[0,1],[6.0003,1],[11.0003,0]]}
          Circle Size {type=MovieSceneScriptingFloatChannel, thinned_from=6, keys [[0,0.5],[5,0.85],[8.0003,0.85],[11.0003,0],[12.0003,0.5]]}
          Inner Glow Start {type=MovieSceneScriptingFloatChannel, keys [[0,0.5],[5,0],[6.0003,-1.25],[11.0003,-1.25],[12.0003,0.5]]}
          Glow Size {type=MovieSceneScriptingFloatChannel, keys [[0,0.5],[2,0.8],[6.0003,0.8],[12.0003,0.5]]}
         ]
        }]
      }]
    }
   ]}
 }
 Anim_Gain {
  binding_count=2
  length_seconds=0.6
  object=&self:Anim_Gain
  sequence {bindings [
    Image_Pip {
     id=#3
     possessed_class=Image
     tracks [{
       type=MovieSceneWidgetMaterialTrack
       sections [{
         type=MovieSceneParameterSection
         channels [
          Stack Opacity {type=MovieSceneScriptingFloatChannel, thinned_from=4, keys [[1,0],[6,1],[12,1]]}
          Circle Size {type=MovieSceneScriptingFloatChannel, thinned_from=6, keys [[0,0.5],[1,0],[4,0.85],[7,0.85],[12,0.5]]}
          Inner Glow Start {type=MovieSceneScriptingFloatChannel, keys [[0,0.5],[1,-1.25],[6,-1.25],[7,0],[12,0.5]]}
          Glow Size {type=MovieSceneScriptingFloatChannel, keys [[0,0.5],[6,0.8],[10,0.8],[12,0.5]]}
         ]
        }]
      }]
    }
    Image_PipFill {
     id=#4
     possessed_class=Image
     tracks [{
       type=MovieSceneFloatTrack
       display_name=Render Opacity
       sections [{type=MovieSceneFloatSection, end_frame=12, start_frame=0, channels [None {type=MovieSceneScriptingFloatChannel, keys [[0,1],[1,0],[12,0]]}]}]
      }]
    }
   ]}
 }
]
graphs [
 EventGraph {nodes [
   Event Construct()
   > SetImagePip!(
    execute
    self:Fort Weapon Reticle Recharge Stack Base Pip Base Object Reference
    Image:Image Object Reference=Get Image_PipFill() -> (Image_PipFill:Image Object Reference)
   ) -> (then)
  ]}
 InitFillState {nodes [
   n1 = InitFillState!() -> (then, IsFilled:bool) {is=title K2Node_FunctionEntry}
   > Parent: InitFillState!(execute, IsFilled:bool=@n1) -> (then) {is=title K2Node_CallParentFunction}
   n2 = Get IsFilled() -> (IsFilled:bool)
   > SetFillPercent!(
    execute
    self:Fort Weapon Reticle Recharge Stack Base Pip Base Object Reference
    FillPercent:float=SelectFloat(A=1, B=0, bPickA=@n2)
    PipPercentMatParamName:Name=Get ImagePipFillParam() -> (ImagePipFillParam:Name=None)
   ) -> (then)
   > n3 = Branch(Condition=@n2)
   n4 = Get Image_Pip() -> (Image_Pip:Image Object Reference)
   > n5 = GetDynamicMaterial(self=@n4)
   n6 = Get Param_StackOpacity() -> (Param_StackOpacity:Name=None)
   > SetScalarParameterValue(self=@n5, ParameterName=@n6, Value=1)
   n7 = Get Image_PipFill()
   > n8 = SetRenderOpacity(self=@n7, InOpacity=0)
   n9 = GetDynamicMaterial(execute=[@n11.then, @n8.then], self=Get Image_Pip())
   > SetScalarParameterValue(self=@n9, ParameterName=Get Param_CircleSize() -> (Param_CircleSize:Name=None), Value=0.5)
   > SetScalarParameterValue(self=@n9, ParameterName=Get Param_InnerGlowStart() -> (Param_InnerGlowStart:Name=None), Value=0.5)
   > SetScalarParameterValue(self=@n9, ParameterName=Get Param_GlowSize() -> (Param_GlowSize:Name=None), Value=0.5)
   n10 = GetDynamicMaterial(execute=@n3.else, self=@n4)
   > SetScalarParameterValue(self=@n10, ParameterName=@n6, Value=0)
   > n11 = SetRenderOpacity(self=@n7, InOpacity=1)
  ]}
 OnFill {nodes [
   OnFill() {is=title K2Node_FunctionEntry}
   > Parent: OnFill() {is=title K2Node_CallParentFunction}
   > SetFillPercent(FillPercent=1, PipPercentMatParamName=Get ImagePipFillParam() -> (ImagePipFillParam=None))
  ]}
 OnUse {nodes [
   OnUse() {is=title K2Node_FunctionEntry}
   > Parent: OnUse() {is=title K2Node_CallParentFunction}
   > SetFillPercent(FillPercent=0, PipPercentMatParamName=Get ImagePipFillParam() -> (ImagePipFillParam=None))
  ]}
 PlayFillAnimation {nodes [
   PlayFillAnimation() {is=title K2Node_FunctionEntry}
   > Parent: PlayFillAnimation() {is=title K2Node_CallParentFunction}
   > StopAnimation(InAnimation=Get Anim_Use() -> (Anim_Use:Widget Animation Object Reference))
   > PlayAnimation(InAnimation=Get Anim_Gain() -> (Anim_Gain:Widget Animation Object Reference))
  ]}
 PlayUseAnimation {nodes [
   PlayUseAnimation() {is=title K2Node_FunctionEntry}
   > Parent: PlayUseAnimation() {is=title K2Node_CallParentFunction}
   > StopAnimation(InAnimation=Get Anim_Gain())
   > PlayAnimation(InAnimation=Get Anim_Use())
  ]}
]
variables [
 ImagePipFillParam {type=string, default=Percent}
 Param_CircleSize {type=string, default=Circle Size}
 Param_GlowSize {type=string, default=Glow Size}
 Param_InnerGlowStart {type=string, default=Inner Glow Start}
 Param_StackOpacity {type=string, default=Stack Opacity}
]
widgets {
 info {
  inheritedWidgetCount=0
  parentClass=/Game/Athena/HUD/reticles/ReticleExtensionWidgets/ChargeStack/WBP_Reticle_RechargeStackBase_Pip.WBP_Reticle_RechargeStackBase_Pip_C
  rootWidgetClass=&Overlay
  widgetCount=4
 }
 widgets [
  Overlay_0 {
   widgetClassPath=&Overlay
   widget=Overlay_0
   properties {renderTransform {angle=0, scale=<0.7 0.7>, shear=<0 0>, translation=<0 0>}}
  }
  Image_Pip {
   widgetClassPath=&Image
   bIsVariable=true
   parent=Overlay_0
   widget=Image_Pip
   properties {
    visibility=SelfHitTestInvisible
    brush {
     bIsDynamicallyLoaded=false
     drawAs=Image
     imageType=FullColor
     mirroring=NoMirror
     resourceName=None
     resourceObject=/WinterDecay/UI/Materials/MI_UI_WinterDecay_RechargeStackPip
     tiling=NoTile
     imageSize=<24 24>
     margin=<ltrb 0 0 0 0>
     outlineSettings {
      bUseBrushTransparency=false
      roundingType=HalfHeightRadius
      width=0
      color {colorUseRule=UseColor_Specified, specifiedColor=<rgba 0 0 0 0>}
      cornerRadii=<0 0 0 0>
     }
     tintColor {colorUseRule=UseColor_Specified, specifiedColor=<rgba 1 1 1 1>}
     uVRegion {bIsValid=false, max=<0 0>, min=<0 0>}
    }
   }
   slot {class=OverlaySlot, properties {horizontalAlignment=HAlign_Fill, verticalAlignment=VAlign_Fill}}
  }
  Image_PipFill {
   widgetClassPath=&Image
   bIsVariable=true
   parent=Overlay_0
   widget=Image_PipFill
   properties {
    visibility=SelfHitTestInvisible
    brush {
     bIsDynamicallyLoaded=false
     drawAs=Image
     imageType=FullColor
     mirroring=NoMirror
     resourceName=None
     resourceObject=/WinterDecay/UI/Materials/M_UI_WinterDecay_RadialProgress
     tiling=NoTile
     imageSize=<24 24>
     margin=<ltrb 0 0 0 0>
     outlineSettings {
      bUseBrushTransparency=false
      roundingType=HalfHeightRadius
      width=0
      color {colorUseRule=UseColor_Specified, specifiedColor=<rgba 0 0 0 0>}
      cornerRadii=<0 0 0 0>
     }
     tintColor {colorUseRule=UseColor_Specified, specifiedColor=<rgba 1 1 1 1>}
     uVRegion {bIsValid=false, max=<0 0>, min=<0 0>}
    }
   }
   slot {class=OverlaySlot, properties {horizontalAlignment=HAlign_Fill, verticalAlignment=VAlign_Fill}}
  }
  Spacer_54 {
   widgetClassPath=&Spacer
   parent=Overlay_0
   widget=Spacer_54
   properties {size=<32 32>}
   slot {class=OverlaySlot, properties {horizontalAlignment=HAlign_Fill, verticalAlignment=VAlign_Fill}}
  }
 ]
}
